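# Animation Blueprint for the NoiseLess backpack cosmetic that adds secondary physics motion to its right and left ring bones.
# The input pose passes in component space through two chained Anim Dynamics nodes, one per ring, then returns to local space for output.
# Each ring uses an angular spring, gravity override, single-axis one-sided swing limit and damped, clamped component motion, tuned differently per side.
# The event graph holds only unconnected stub nodes.
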AnimBlueprint /BRCosmetics/Accessories/FORT_Backpacks/Backpack_F_MED_NoiseLess/Meshes/F_MED_NoiseLess_Pack_AnimBP
target_skeleton=&/F_MED_NoiseLess_Pack_Skeleton
graphs [
 AnimGraph {nodes [
   n1 = Anim Dynamics - Bone: dyn_ring_r(ComponentPose=Local To Component(LocalPose=Input Pose()), Alpha=1, ExternalForce="0,0,0") {
    angularSpringConstant=10
    bAngularSpring=true
    bUseGravityOverride=true
    gravityScale=0.35
    alphaBoolBlend {blendOption=Linear}
    boundBone {boneName=dyn_ring_r}
    componentAppliedLinearAccClamp=<600 600 600>
    componentLinearAccScale=<0.15 0.15 0.15>
    componentLinearVelScale=<0.2 0.2 0.2>
    gravityOverride {z=-200}
    physicsBodyDefinitions [{
      sphereCollisionRadius=0
      boundBone {boneName=dyn_ring_r}
      boxExtents=<8 8 10>
      constraintSetup {angularTargetAxis=AxisY, angularLimitsMax=<0 0 0>, angularLimitsMin=<0 -90 0>, angularTarget=<0 0 0>}
      localJointOffset=<5 0 0>
     }]
   }
   n2 = Anim Dynamics - Bone: dyn_ring_l(ComponentPose=@n1, Alpha=1, ExternalForce="0,0,0") {
    angularSpringConstant=10
    bAngularSpring=true
    bUseGravityOverride=true
    gravityScale=0.35
    alphaBoolBlend {blendOption=Linear}
    boundBone {boneName=dyn_ring_l}
    componentAppliedLinearAccClamp=<800 800 800>
    componentLinearAccScale=<0.17 0.17 0.17>
    componentLinearVelScale=<0.15 0.15 0.15>
    gravityOverride {z=-300}
    physicsBodyDefinitions [{
      sphereCollisionRadius=0
      boundBone {boneName=dyn_ring_l}
      boxExtents=<7 8 10>
      constraintSetup {angularTargetAxis=AxisY, angularLimitsMax=<0 0 0>, angularLimitsMin=<0 -100 0>, angularTarget=<0 0 0>}
      localJointOffset=<4 0 0>
     }]
   }
   Output Pose(Result=Component To Local(ComponentPose=@n2)) {layerGroup=None}
  ]}
 EventGraph {nodes [
   TryGetPawnOwner()
   Event BlueprintUpdateAnimation()
  ]}
]
variables []
